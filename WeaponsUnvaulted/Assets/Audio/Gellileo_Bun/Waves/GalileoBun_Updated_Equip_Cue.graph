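# Sound cue for the equip sound of the Galileo Bun weapon, classed as Fort weapon foley.
# It reuses the weapon's reload wave, played through a short fixed delay.
# The cue lowers the pitch slightly and uses the Fort weapon reload attenuation for distance falloff.
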SoundCue /WeaponsUnvaulted/Assets/Audio/Gellileo_Bun/Waves/GalileoBun_Updated_Equip_Cue
root=Delay_0
nodes [
 {id=Delay_0, children [WavePlayer_0], properties {delayMax=0.04, delayMin=0.04}}
 {id=WavePlayer_0, properties {soundWaveAssetPtr=&/GalileoBun_Updated_Reload}}
]
properties {
 attenuationSettings=&Fort_Weapon_Reload_Att
 maxDistance=2250
 pitchMultiplier=0.95
 soundClassObject=&Fort_WeaponFoley
}
